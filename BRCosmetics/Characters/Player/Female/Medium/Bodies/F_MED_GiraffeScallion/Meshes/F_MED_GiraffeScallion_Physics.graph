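# Physics asset for the GiraffeScallion female medium player body, covering its dynamic skirt, skirt fluff and belt bones.
# Five four-link skirt chains and two two-link fluff chains hang from MSTR bodies on spine_01, with tight swing and twist limits per link.
# Eight belt bodies attach to the pelvis, all angularly free except the right one.
# Torso, leg and foot capsules join a huge root capsule sunk below the origin, apparently a floor collider.
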PhysicsAsset /BRCosmetics/Characters/Player/Female/Medium/Bodies/F_MED_GiraffeScallion/Meshes/F_MED_GiraffeScallion_Physics
bodies {
 calf_l [{shapeType=Capsule, length=30, radius=7, center=<-21 -3 -1>, rotation=<rot 0 90 90>}]
 calf_r [{shapeType=Capsule, length=30, radius=7, center=<21 3 1>, rotation=<rot 0 90 90>}]
 dyn_Skirt_bk_c_01_jnt [{shapeType=Capsule, length=8, radius=7, shapeName=dyn_Skirt_bk_c_01_jnt_capsule, center=<10 -2 0>, rotation=<rot 0 90 90>}]
 dyn_Skirt_bk_c_02_jnt [{shapeType=Capsule, length=8, radius=7, center=<10 -2 0>, rotation=<rot 0 90 90>}]
 dyn_Skirt_bk_c_03_jnt [{shapeType=Capsule, length=8, radius=7, shapeName=dyn_Skirt_bk_c_03_jnt_capsule, center=<10 -2 0>, rotation=<rot 0 90 90>}]
 dyn_Skirt_bk_c_04_jnt [{shapeType=Capsule, length=8, radius=7, center=<10 -2 0>, rotation=<rot 0 90 90>}]
 dyn_Skirt_bk_c_MSTR [{shapeType=Capsule, length=0.505, radius=0.505, shapeName=dyn_Skirt_bk_c_MSTR_capsule, rotation=<rot 0 -90 0>}]
 dyn_Skirt_bk_l_01_jnt [{shapeType=Capsule, length=8, radius=7, center=<10 -2 0>, rotation=<rot 0 90 90>}]
 dyn_Skirt_bk_l_02_jnt [{shapeType=Capsule, length=8, radius=7, shapeName=dyn_Skirt_bk_l_02_jnt_capsule, center=<10 -2 0>, rotation=<rot 0 90 90>}]
 dyn_Skirt_bk_l_03_jnt [{shapeType=Capsule, length=8, radius=7, shapeName=dyn_Skirt_bk_l_03_jnt_capsule, center=<10 -2 0>, rotation=<rot 0 90 90>}]
 dyn_Skirt_bk_l_04_jnt [{shapeType=Capsule, length=8, radius=7, shapeName=dyn_Skirt_bk_l_04_jnt_capsule, center=<10 -2 0>, rotation=<rot 0 90 90>}]
 dyn_Skirt_bk_l_MSTR [{shapeType=Capsule, length=0.505, radius=0.505, shapeName=dyn_Skirt_bk_l_MSTR_capsule, rotation=<rot 0 -90 0>}]
 dyn_Skirt_bk_r_01_jnt [{shapeType=Capsule, length=8, radius=7, center=<10 -2 0>, rotation=<rot 0 90 90>}]
 dyn_Skirt_bk_r_02_jnt [{shapeType=Capsule, length=8, radius=7, shapeName=dyn_Skirt_bk_r_02_jnt_capsule, center=<10 -2 0>, rotation=<rot 0 90 90>}]
 dyn_Skirt_bk_r_03_jnt [{shapeType=Capsule, length=8, radius=7, shapeName=dyn_Skirt_bk_r_03_jnt_capsule, center=<10 -2 0>, rotation=<rot 0 90 90>}]
 dyn_Skirt_bk_r_04_jnt [{shapeType=Capsule, length=8, radius=7, shapeName=dyn_Skirt_bk_r_04_jnt_capsule, center=<10 -2 0>, rotation=<rot 0 90 90>}]
 dyn_Skirt_bk_r_MSTR [{shapeType=Capsule, length=0.505, radius=0.505, shapeName=dyn_Skirt_bk_r_MSTR_capsule, rotation=<rot 0 -90 0>}]
 dyn_Skirt_l_01_jnt [{shapeType=Capsule, length=8, radius=7, shapeName=dyn_Skirt_l_01_jnt_capsule, center=<10 -2 0>, rotation=<rot 0 90 90>}]
 dyn_Skirt_l_02_jnt [{shapeType=Capsule, length=8, radius=7, shapeName=dyn_Skirt_l_02_jnt_capsule, center=<10 -2 0>, rotation=<rot 0 90 90>}]
 dyn_Skirt_l_03_jnt [{shapeType=Capsule, length=8, radius=7, center=<10 -2 0>, rotation=<rot 0 90 90>}]
 dyn_Skirt_l_04_jnt [{shapeType=Capsule, length=8, radius=7, shapeName=dyn_Skirt_l_04_jnt_capsule, center=<10 -2 0>, rotation=<rot 0 90 90>}]
 dyn_Skirt_l_MSTR [{shapeType=Capsule, length=0.505, radius=0.505, shapeName=dyn_Skirt_l_MSTR_capsule, rotation=<rot 0 -90 0>}]
 dyn_Skirt_r_01_jnt [{shapeType=Capsule, length=8, radius=7, shapeName=dyn_Skirt_r_01_jnt_capsule, center=<10 -2 0>, rotation=<rot 0 90 90>}]
 dyn_Skirt_r_02_jnt [{shapeType=Capsule, length=8, radius=7, shapeName=dyn_Skirt_r_02_jnt_capsule, center=<10 -2 0>, rotation=<rot 0 90 90>}]
 dyn_Skirt_r_03_jnt [{shapeType=Capsule, length=8, radius=7, center=<10 -2 0>, rotation=<rot 0 90 90>}]
 dyn_Skirt_r_04_jnt [{shapeType=Capsule, length=8, radius=7, shapeName=dyn_Skirt_r_04_jnt_capsule, center=<10 -2 0>, rotation=<rot 0 90 90>}]
 dyn_Skirt_r_MSTR [{shapeType=Capsule, length=0.505, radius=0.505, shapeName=dyn_Skirt_r_MSTR_capsule, rotation=<rot 0 -90 0>}]
 dyn_bk_belt_jnt [{
   shapeType=Capsule
   length=3.9996
   radius=3.3745
   shapeName=dyn_bk_belt_jnt_capsule
   center=<-0.0702 -0.4535 -0.8101>
   rotation=<rot 0 -107.657 -75.3805>
  }]
 dyn_bk_l_belt_jnt [{
   shapeType=Capsule
   length=4.1839
   radius=2.3445
   shapeName=dyn_bk_l_belt_jnt_capsule
   center=<3.2833 0.0322 -0.0321>
   rotation=<rot 1.0463 -90 40.1711>
  }]
 dyn_bk_r_belt_jnt [{
   shapeType=Capsule
   length=7.1636
   radius=3.7916
   shapeName=dyn_bk_r_belt_jnt_capsule
   center=<-4.6904 0.6552 -0.9079>
   rotation=<rot 0 73.1588 -83.001>
  }]
 dyn_ft_belt_jnt [{
   shapeType=Capsule
   length=5.4735
   radius=3.136
   shapeName=dyn_ft_belt_jnt_capsule
   center=<0.4784 -0.2196 -0.4581>
   rotation=<rot 0 -94.9558 -69.4117>
  }]
 dyn_ft_l_belt_jnt [{
   shapeType=Capsule
   length=6.7954
   radius=3.0529
   shapeName=dyn_ft_l_belt_jnt_capsule
   center=<2.0485 -0.2135 0.0578>
   rotation=<rot 0 87.3519 -86.1179>
  }]
 dyn_ft_r_belt_jnt [{
   shapeType=Capsule
   length=5.94
   radius=3.9881
   shapeName=dyn_ft_r_belt_jnt_capsule
   center=<-4.0271 0.5007 -0.0659>
   rotation=<rot 0 -98.6237 -75.1064>
  }]
 dyn_l_belt_jnt [{
   shapeType=Capsule
   length=9.6572
   radius=2.8727
   shapeName=dyn_l_belt_jnt_capsule
   center=<-2.8491 3.231 -0.7794>
   rotation=<rot 0 18.3606 -84.7846>
  }]
 dyn_r_belt_jnt [{shapeType=Capsule, length=17.767, radius=9.805, center=<2.2955 -0.7517 0.2098>, rotation=<rot -21.2515 -90 10.888>}]
 dyn_skirtfluff_l_01_jnt [{shapeType=Capsule, length=12, radius=3, shapeName=dyn_skirtfluff_l_01_jnt_capsule, center=<3 -1 0>}]
 dyn_skirtfluff_l_02_jnt [{shapeType=Capsule, length=12, radius=3, shapeName=dyn_skirtfluff_l_02_jnt_capsule, center=<3 -1 0>}]
 dyn_skirtfluff_l_MSTR [{shapeType=Capsule, length=0.505, radius=0.505, shapeName=dyn_skirtfluff_l_MSTR_capsule, rotation=<rot 0 -90 0>}]
 dyn_skirtfluff_r_01_jnt [{shapeType=Capsule, length=12, radius=3, shapeName=dyn_skirtfluff_r_01_jnt_capsule, center=<3 -1 0>}]
 dyn_skirtfluff_r_02_jnt [{shapeType=Capsule, length=12, radius=3, shapeName=dyn_skirtfluff_r_02_jnt_capsule, center=<3 -1 0>}]
 dyn_skirtfluff_r_MSTR [{shapeType=Capsule, length=0.505, radius=0.505, shapeName=dyn_skirtfluff_r_MSTR_capsule, rotation=<rot 0 -90 0>}]
 foot_l [{shapeType=Capsule, length=15, radius=6.4, center=<-4.3 6 0>, rotation=<rot 0 0 90>}]
 foot_r [{shapeType=Capsule, length=15, radius=6.4, center=<4.3 -6 0>, rotation=<rot 0 0 90>}]
 pelvis [{shapeType=Capsule, length=12, radius=10}]
 root [{shapeType=Capsule, length=60, radius=200, shapeName=root_capsule, center=<0 0 -185>, rotation=<rot 0 90 90>}]
 spine_01 [{shapeType=Capsule, length=6, radius=8}]
 thigh_l [{shapeType=Capsule, length=27, radius=7, center=<-21 0 0>, rotation=<rot 0 90 90>}]
 thigh_r [{shapeType=Capsule, length=27, radius=7, center=<21 0 0>, rotation=<rot 0 90 90>}]
}
constraints [
 {bone1Name=spine_01, bone2Name=pelvis, swing1LimitDegrees=45, swing2LimitDegrees=45, twistLimitDegrees=45}
 {bone1Name=dyn_Skirt_r_03_jnt, bone2Name=dyn_Skirt_r_02_jnt, swing1LimitDegrees=21, swing2LimitDegrees=5, twistLimitDegrees=1}
 {bone1Name=dyn_Skirt_bk_r_01_jnt, bone2Name=dyn_Skirt_bk_r_MSTR, swing1LimitDegrees=21, swing2LimitDegrees=5, twistLimitDegrees=1}
 {bone1Name=dyn_Skirt_bk_c_02_jnt, bone2Name=dyn_Skirt_bk_c_01_jnt, swing1LimitDegrees=21, swing2LimitDegrees=5, twistLimitDegrees=1}
 {bone1Name=dyn_Skirt_bk_c_04_jnt, bone2Name=dyn_Skirt_bk_c_03_jnt, swing1LimitDegrees=21, swing2LimitDegrees=5, twistLimitDegrees=1}
 {bone1Name=dyn_Skirt_l_03_jnt, bone2Name=dyn_Skirt_l_02_jnt, swing1LimitDegrees=21, swing2LimitDegrees=5, twistLimitDegrees=1}
 {bone1Name=dyn_Skirt_bk_l_01_jnt, bone2Name=dyn_Skirt_bk_l_MSTR, swing1LimitDegrees=21, swing2LimitDegrees=5, twistLimitDegrees=1}
 {bone1Name=thigh_l, bone2Name=pelvis, swing1LimitDegrees=45, swing2LimitDegrees=45, twistLimitDegrees=45}
 {bone1Name=calf_l, bone2Name=thigh_l, swing1LimitDegrees=45, swing2LimitDegrees=45, twistLimitDegrees=45}
 {bone1Name=thigh_r, bone2Name=pelvis, swing1LimitDegrees=45, swing2LimitDegrees=45, twistLimitDegrees=45}
 {bone1Name=calf_r, bone2Name=thigh_r, swing1LimitDegrees=45, swing2LimitDegrees=45, twistLimitDegrees=45}
 {bone1Name=dyn_r_belt_jnt, bone2Name=pelvis, swing1LimitDegrees=45, swing2LimitDegrees=45, twistLimitDegrees=45}
 {
  bone1Name=dyn_bk_belt_jnt
  bone2Name=pelvis
  swing1LimitDegrees=45
  swing1Motion=Free
  swing2LimitDegrees=45
  swing2Motion=Free
  twistLimitDegrees=45
  twistMotion=Free
 }
 {
  bone1Name=dyn_bk_l_belt_jnt
  bone2Name=pelvis
  swing1LimitDegrees=45
  swing1Motion=Free
  swing2LimitDegrees=45
  swing2Motion=Free
  twistLimitDegrees=45
  twistMotion=Free
 }
 {
  bone1Name=dyn_bk_r_belt_jnt
  bone2Name=pelvis
  swing1LimitDegrees=45
  swing1Motion=Free
  swing2LimitDegrees=45
  swing2Motion=Free
  twistLimitDegrees=45
  twistMotion=Free
 }
 {
  bone1Name=dyn_ft_belt_jnt
  bone2Name=pelvis
  swing1LimitDegrees=45
  swing1Motion=Free
  swing2LimitDegrees=45
  swing2Motion=Free
  twistLimitDegrees=45
  twistMotion=Free
 }
 {
  bone1Name=dyn_ft_l_belt_jnt
  bone2Name=pelvis
  swing1LimitDegrees=45
  swing1Motion=Free
  swing2LimitDegrees=45
  swing2Motion=Free
  twistLimitDegrees=45
  twistMotion=Free
 }
 {
  bone1Name=dyn_ft_r_belt_jnt
  bone2Name=pelvis
  swing1LimitDegrees=45
  swing1Motion=Free
  swing2LimitDegrees=45
  swing2Motion=Free
  twistLimitDegrees=45
  twistMotion=Free
 }
 {
  bone1Name=dyn_l_belt_jnt
  bone2Name=pelvis
  swing1LimitDegrees=45
  swing1Motion=Free
  swing2LimitDegrees=45
  swing2Motion=Free
  twistLimitDegrees=45
  twistMotion=Free
 }
 {
  bone1Name=dyn_Skirt_bk_c_MSTR
  bone2Name=spine_01
  swing1LimitDegrees=45
  swing1Motion=Free
  swing2LimitDegrees=45
  swing2Motion=Free
  twistLimitDegrees=45
  twistMotion=Free
 }
 {bone1Name=dyn_Skirt_bk_c_01_jnt, bone2Name=dyn_Skirt_bk_c_MSTR, swing1LimitDegrees=21, swing2LimitDegrees=5, twistLimitDegrees=1}
 {bone1Name=dyn_Skirt_bk_c_03_jnt, bone2Name=dyn_Skirt_bk_c_02_jnt, swing1LimitDegrees=21, swing2LimitDegrees=5, twistLimitDegrees=1}
 {
  bone1Name=dyn_Skirt_bk_l_MSTR
  bone2Name=spine_01
  swing1LimitDegrees=45
  swing1Motion=Free
  swing2LimitDegrees=45
  swing2Motion=Free
  twistLimitDegrees=45
  twistMotion=Free
 }
 {bone1Name=dyn_Skirt_bk_l_02_jnt, bone2Name=dyn_Skirt_bk_l_01_jnt, swing1LimitDegrees=21, swing2LimitDegrees=5, twistLimitDegrees=1}
 {bone1Name=dyn_Skirt_bk_l_03_jnt, bone2Name=dyn_Skirt_bk_l_02_jnt, swing1LimitDegrees=21, swing2LimitDegrees=5, twistLimitDegrees=1}
 {bone1Name=dyn_Skirt_bk_l_04_jnt, bone2Name=dyn_Skirt_bk_l_03_jnt, swing1LimitDegrees=21, swing2LimitDegrees=5, twistLimitDegrees=1}
 {
  bone1Name=dyn_Skirt_bk_r_MSTR
  bone2Name=spine_01
  swing1LimitDegrees=45
  swing1Motion=Free
  swing2LimitDegrees=45
  swing2Motion=Free
  twistLimitDegrees=45
  twistMotion=Free
 }
 {bone1Name=dyn_Skirt_bk_r_02_jnt, bone2Name=dyn_Skirt_bk_r_01_jnt, swing1LimitDegrees=21, swing2LimitDegrees=5, twistLimitDegrees=1}
 {bone1Name=dyn_Skirt_bk_r_03_jnt, bone2Name=dyn_Skirt_bk_r_02_jnt, swing1LimitDegrees=21, swing2LimitDegrees=5, twistLimitDegrees=1}
 {bone1Name=dyn_Skirt_bk_r_04_jnt, bone2Name=dyn_Skirt_bk_r_03_jnt, swing1LimitDegrees=21, swing2LimitDegrees=5, twistLimitDegrees=1}
 {
  bone1Name=dyn_Skirt_l_MSTR
  bone2Name=spine_01
  swing1LimitDegrees=45
  swing1Motion=Free
  swing2LimitDegrees=45
  swing2Motion=Free
  twistLimitDegrees=45
  twistMotion=Free
 }
 {bone1Name=dyn_Skirt_l_01_jnt, bone2Name=dyn_Skirt_l_MSTR, swing1LimitDegrees=21, swing2LimitDegrees=5, twistLimitDegrees=1}
 {bone1Name=dyn_Skirt_l_02_jnt, bone2Name=dyn_Skirt_l_01_jnt, swing1LimitDegrees=21, swing2LimitDegrees=5, twistLimitDegrees=1}
 {bone1Name=dyn_Skirt_l_04_jnt, bone2Name=dyn_Skirt_l_03_jnt, swing1LimitDegrees=21, swing2LimitDegrees=5, twistLimitDegrees=1}
 {
  bone1Name=dyn_Skirt_r_MSTR
  bone2Name=spine_01
  swing1LimitDegrees=45
  swing1Motion=Free
  swing2LimitDegrees=45
  swing2Motion=Free
  twistLimitDegrees=45
  twistMotion=Free
 }
 {bone1Name=dyn_Skirt_r_01_jnt, bone2Name=dyn_Skirt_r_MSTR, swing1LimitDegrees=21, swing2LimitDegrees=5, twistLimitDegrees=1}
 {bone1Name=dyn_Skirt_r_02_jnt, bone2Name=dyn_Skirt_r_01_jnt, swing1LimitDegrees=21, swing2LimitDegrees=5, twistLimitDegrees=1}
 {bone1Name=dyn_Skirt_r_04_jnt, bone2Name=dyn_Skirt_r_03_jnt, swing1LimitDegrees=21, swing2LimitDegrees=5, twistLimitDegrees=1}
 {
  bone1Name=dyn_skirtfluff_l_MSTR
  bone2Name=spine_01
  swing1LimitDegrees=45
  swing1Motion=Free
  swing2LimitDegrees=45
  swing2Motion=Free
  twistLimitDegrees=45
  twistMotion=Free
 }
 {bone1Name=dyn_skirtfluff_l_01_jnt, bone2Name=dyn_skirtfluff_l_MSTR, swing1LimitDegrees=15, swing2LimitDegrees=5, twistLimitDegrees=1}
 {bone1Name=dyn_skirtfluff_l_02_jnt, bone2Name=dyn_skirtfluff_l_01_jnt, swing1LimitDegrees=15, swing2LimitDegrees=5, twistLimitDegrees=1}
 {
  bone1Name=dyn_skirtfluff_r_MSTR
  bone2Name=spine_01
  swing1LimitDegrees=45
  swing1Motion=Free
  swing2LimitDegrees=45
  swing2Motion=Free
  twistLimitDegrees=45
  twistMotion=Free
 }
 {bone1Name=dyn_skirtfluff_r_01_jnt, bone2Name=dyn_skirtfluff_r_MSTR, swing1LimitDegrees=15, swing2LimitDegrees=5, twistLimitDegrees=1}
 {bone1Name=dyn_skirtfluff_r_02_jnt, bone2Name=dyn_skirtfluff_r_01_jnt, swing1LimitDegrees=15, swing2LimitDegrees=5, twistLimitDegrees=1}
]
