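# Physics asset for the GiraffeScallion female medium hair pin face accessory, defining bodies and constraints for dynamic hair bones.
# Four capsule chains (bottom-left, top-left, right and right bang) each link a small master body to the head with free swing and twist.
# Joints down each chain have narrow swing and very tight twist limits.
# Head, neck, upper spine, clavicle and upper-arm capsules provide body shapes around the hair.
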PhysicsAsset /BRCosmetics/Characters/Player/Female/Medium/Bodies/F_MED_GiraffeScallion/Meshes/Parts/F_MED_GiraffeScallion_FaceAcc_HairPin_Physics
bodies {
 clavicle_l [{shapeType=Capsule, length=13, radius=5.75, shapeName=clavicle_l_capsule, center=<3.5 -1 -1>, rotation=<rot 80 -0.0003 -0.0003>}]
 clavicle_r [{shapeType=Capsule, length=13, radius=5.75, shapeName=clavicle_r_capsule, center=<-3.5 1 1>, rotation=<rot 80 -0.0003 -0.0004>}]
 dyn_hair_bt_l_01_jnt [{shapeType=Capsule, length=2.5, radius=2, shapeName=dyn_hair_bt_l_01_jnt_capsule, center=<2.25 1 -0.5>, rotation=<rot 0 90 90>}]
 dyn_hair_bt_l_02_jnt [{shapeType=Capsule, length=2, radius=1, shapeName=dyn_hair_bt_l_02_jnt_capsule, center=<2 0.2 0>, rotation=<rot 0.0004 119.9999 89.9999>}]
 dyn_hair_bt_l_03_jnt [{shapeType=Capsule, length=1, radius=0.7, shapeName=dyn_hair_bt_l_03_jnt_capsule, center=<1.5 0.1 0>, rotation=<rot 0.0001 99.9999 90.0001>}]
 dyn_hair_bt_l_04_jnt [{shapeType=Capsule, length=0.5, radius=0.5, shapeName=dyn_hair_bt_l_04_jnt_capsule, center=<1 0 0>, rotation=<rot 0 93.2687 -89.0029>}]
 dyn_hair_bt_l_MSTR [{shapeType=Capsule, length=0.505, radius=0.505, shapeName=dyn_hair_bt_l_MSTR_capsule, rotation=<rot 0 -90 0>}]
 dyn_hair_r_01_jnt [{shapeType=Capsule, length=3, radius=3, shapeName=dyn_hair_r_01_jnt_capsule, center=<2 0 0>, rotation=<rot 0 90 90>}]
 dyn_hair_r_02_jnt [{shapeType=Capsule, length=3, radius=3, center=<2 0 0.5>, rotation=<rot 0 90 90>}]
 dyn_hair_r_03_jnt [{shapeType=Capsule, length=3, radius=3, shapeName=dyn_hair_r_03_jnt_capsule, center=<2 1 0.5>, rotation=<rot 0 89.9999 105.0001>}]
 dyn_hair_r_04_jnt [{shapeType=Capsule, length=3, radius=2.5, shapeName=dyn_hair_r_04_jnt_capsule, center=<2 1 -0.5>, rotation=<rot 0 90 90>}]
 dyn_hair_r_05_jnt [{shapeType=Capsule, length=3, radius=1.5, shapeName=dyn_hair_r_05_jnt_capsule, center=<2 0 0>, rotation=<rot 0 90 95>}]
 dyn_hair_r_06_jnt [{shapeType=Capsule, length=3, radius=1, shapeName=dyn_hair_r_06_jnt_capsule, center=<2 0 0.5>, rotation=<rot -0.0001 90.0001 110.0001>}]
 dyn_hair_r_MSTR [{shapeType=Capsule, length=0.505, radius=0.505, shapeName=dyn_hair_r_MSTR_capsule, rotation=<rot 0 -90 0>}]
 dyn_hair_tp_l_01_jnt [{shapeType=Capsule, length=2, radius=3, shapeName=dyn_hair_tp_l_01_jnt_capsule, center=<2 1 0>, rotation=<rot 0 90 90>}]
 dyn_hair_tp_l_02_jnt [{shapeType=Capsule, length=2, radius=3, shapeName=dyn_hair_tp_l_02_jnt_capsule, center=<2 1 0>, rotation=<rot 0 90 90>}]
 dyn_hair_tp_l_03_jnt [{shapeType=Capsule, length=2, radius=3, shapeName=dyn_hair_tp_l_03_jnt_capsule, center=<2 1 0>, rotation=<rot 0 90 90>}]
 dyn_hair_tp_l_04_jnt [{shapeType=Capsule, length=2, radius=1.5, shapeName=dyn_hair_tp_l_04_jnt_capsule, center=<2 1 0>, rotation=<rot 0 90 90>}]
 dyn_hair_tp_l_05_jnt [{shapeType=Capsule, length=1, radius=1, shapeName=dyn_hair_tp_l_05_jnt_capsule, center=<1 1 0>, rotation=<rot 0 90 90>}]
 dyn_hair_tp_l_MSTR [{shapeType=Capsule, length=0.505, radius=0.505, shapeName=dyn_hair_tp_l_MSTR_capsule, rotation=<rot 0 -90 0>}]
 dyn_hairbang_r_01_jnt [{shapeType=Capsule, length=3, radius=2, shapeName=dyn_hairbang_r_01_jnt_capsule, center=<3 0 1>, rotation=<rot 0 90 90>}]
 dyn_hairbang_r_02_jnt [{shapeType=Capsule, length=3, radius=2, shapeName=dyn_hairbang_r_02_jnt_capsule, center=<3 0 1>, rotation=<rot 0 90 95>}]
 dyn_hairbang_r_03_jnt [{shapeType=Capsule, length=3, radius=2, shapeName=dyn_hairbang_r_03_jnt_capsule, center=<3 0 1>, rotation=<rot 0 90 90>}]
 dyn_hairbang_r_04_jnt [{shapeType=Capsule, length=3, radius=1.5, shapeName=dyn_hairbang_r_04_jnt_capsule, center=<3 0 0.5>, rotation=<rot 0 90 90>}]
 dyn_hairbang_r_05_jnt [{shapeType=Capsule, length=1, radius=1, shapeName=dyn_hairbang_r_05_jnt_capsule, center=<2 0 0>, rotation=<rot 0 90 90>}]
 dyn_hairbang_r_06_jnt [{shapeType=Capsule, length=1, radius=0.5, shapeName=dyn_hairbang_r_06_jnt_capsule, center=<1.5 -0.5 0>, rotation=<rot 0.0001 64.9999 90>}]
 dyn_hairbang_r_MSTR [{shapeType=Capsule, length=0.505, radius=0.505, shapeName=dyn_hairbang_r_MSTR_capsule, rotation=<rot 0 -90 0>}]
 head [{shapeType=Capsule, length=6, radius=6, shapeName=head_capsule, center=<3.5 3 0>, rotation=<rot 90 0 0>}]
 neck_01 [{shapeType=Capsule, length=8, radius=4.5, shapeName=neck_01_capsule, center=<3.6683 -0.2876 0>, rotation=<rot 90 0 0>}]
 spine_05 [
  {shapeType=Capsule, length=6, radius=9.25, shapeName=spine_05_capsule, center=<0.1755 0.9721 0>}
  {shapeType=Capsule, length=10, radius=7.75, shapeName=spine_05_capsule, center=<-2.1689 5.0823 0>}
 ]
 upperarm_l [{shapeType=Capsule, length=20, radius=3.5, shapeName=upperarm_l_capsule, center=<12 -0.5 0>, rotation=<rot 90 0 0>}]
 upperarm_r [{shapeType=Capsule, length=20, radius=3.5, shapeName=upperarm_r_capsule, center=<-12 0.5 0>, rotation=<rot 90 0 0>}]
}
constraints [
 {bone1Name=dyn_hair_r_02_jnt, bone2Name=dyn_hair_r_01_jnt, swing1LimitDegrees=10, swing2LimitDegrees=10, twistLimitDegrees=1}
 {
  bone1Name=dyn_hair_bt_l_MSTR
  bone2Name=head
  swing1LimitDegrees=45
  swing1Motion=Free
  swing2LimitDegrees=45
  swing2Motion=Free
  twistLimitDegrees=45
  twistMotion=Free
 }
 {bone1Name=dyn_hair_bt_l_01_jnt, bone2Name=dyn_hair_bt_l_MSTR, swing1LimitDegrees=15, swing2LimitDegrees=15, twistLimitDegrees=1}
 {bone1Name=dyn_hair_bt_l_02_jnt, bone2Name=dyn_hair_bt_l_01_jnt, swing1LimitDegrees=15, swing2LimitDegrees=15, twistLimitDegrees=1}
 {bone1Name=dyn_hair_bt_l_03_jnt, bone2Name=dyn_hair_bt_l_02_jnt, swing1LimitDegrees=15, swing2LimitDegrees=15, twistLimitDegrees=1}
 {bone1Name=dyn_hair_bt_l_04_jnt, bone2Name=dyn_hair_bt_l_03_jnt, swing1LimitDegrees=15, swing2LimitDegrees=15, twistLimitDegrees=1}
 {
  bone1Name=dyn_hair_r_MSTR
  bone2Name=head
  swing1LimitDegrees=45
  swing1Motion=Free
  swing2LimitDegrees=45
  swing2Motion=Free
  twistLimitDegrees=45
  twistMotion=Free
 }
 {bone1Name=dyn_hair_r_01_jnt, bone2Name=dyn_hair_r_MSTR, swing1LimitDegrees=10, swing2LimitDegrees=10, twistLimitDegrees=1}
 {bone1Name=dyn_hair_r_03_jnt, bone2Name=dyn_hair_r_02_jnt, swing1LimitDegrees=10, swing2LimitDegrees=10, twistLimitDegrees=1}
 {bone1Name=dyn_hair_r_04_jnt, bone2Name=dyn_hair_r_03_jnt, swing1LimitDegrees=10, swing2LimitDegrees=10, twistLimitDegrees=1}
 {bone1Name=dyn_hair_r_05_jnt, bone2Name=dyn_hair_r_04_jnt, swing1LimitDegrees=10, swing2LimitDegrees=10, twistLimitDegrees=1}
 {bone1Name=dyn_hair_r_06_jnt, bone2Name=dyn_hair_r_05_jnt, swing1LimitDegrees=10, swing2LimitDegrees=10, twistLimitDegrees=1}
 {
  bone1Name=dyn_hair_tp_l_MSTR
  bone2Name=head
  swing1LimitDegrees=45
  swing1Motion=Free
  swing2LimitDegrees=45
  swing2Motion=Free
  twistLimitDegrees=45
  twistMotion=Free
 }
 {bone1Name=dyn_hair_tp_l_01_jnt, bone2Name=dyn_hair_tp_l_MSTR, swing1LimitDegrees=15, swing2LimitDegrees=15, twistLimitDegrees=1}
 {bone1Name=dyn_hair_tp_l_02_jnt, bone2Name=dyn_hair_tp_l_01_jnt, swing1LimitDegrees=15, swing2LimitDegrees=15, twistLimitDegrees=1}
 {bone1Name=dyn_hair_tp_l_03_jnt, bone2Name=dyn_hair_tp_l_02_jnt, swing1LimitDegrees=15, swing2LimitDegrees=15, twistLimitDegrees=1}
 {bone1Name=dyn_hair_tp_l_04_jnt, bone2Name=dyn_hair_tp_l_03_jnt, swing1LimitDegrees=15, swing2LimitDegrees=15, twistLimitDegrees=1}
 {bone1Name=dyn_hair_tp_l_05_jnt, bone2Name=dyn_hair_tp_l_04_jnt, swing1LimitDegrees=15, swing2LimitDegrees=15, twistLimitDegrees=1}
 {
  bone1Name=dyn_hairbang_r_MSTR
  bone2Name=head
  swing1LimitDegrees=45
  swing1Motion=Free
  swing2LimitDegrees=45
  swing2Motion=Free
  twistLimitDegrees=45
  twistMotion=Free
 }
 {bone1Name=dyn_hairbang_r_01_jnt, bone2Name=dyn_hairbang_r_MSTR, swing1LimitDegrees=10, swing2LimitDegrees=10, twistLimitDegrees=1}
 {bone1Name=dyn_hairbang_r_02_jnt, bone2Name=dyn_hairbang_r_01_jnt, swing1LimitDegrees=10, swing2LimitDegrees=10, twistLimitDegrees=1}
 {bone1Name=dyn_hairbang_r_03_jnt, bone2Name=dyn_hairbang_r_02_jnt, swing1LimitDegrees=10, swing2LimitDegrees=10, twistLimitDegrees=1}
 {bone1Name=dyn_hairbang_r_04_jnt, bone2Name=dyn_hairbang_r_03_jnt, swing1LimitDegrees=10, swing2LimitDegrees=10, twistLimitDegrees=1}
 {bone1Name=dyn_hairbang_r_05_jnt, bone2Name=dyn_hairbang_r_04_jnt, swing1LimitDegrees=10, swing2LimitDegrees=10, twistLimitDegrees=1}
 {bone1Name=dyn_hairbang_r_06_jnt, bone2Name=dyn_hairbang_r_05_jnt, swing1LimitDegrees=10, swing2LimitDegrees=10, twistLimitDegrees=1}
]
